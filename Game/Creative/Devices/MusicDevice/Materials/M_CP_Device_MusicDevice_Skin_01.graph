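# Default Lit skin material for the Creative Music Device, with pulsing speakers and tintable color and glow.
# A time-driven sine, clamped to zero or below, scales the pivot-to-vertex vector into an outward pulse, halved on Y and masked by vertex color.
# Diffuse alpha masks tinted speaker emissive, while a Turn On parameter switches both pulse and glow.
# Base color tint fades by texture brightness per channel, alongside normal and mask textures.
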Material /Game/Creative/Devices/MusicDevice/Materials/M_CP_Device_MusicDevice_Skin_01
graph {
 properties {pixelDepthOffsetMode=PDOM_Legacy, usedShadingModels=Default Lit}
 nodes [
  n1 = Time()
  n2 = ScalarParameter() {defaultValue=1.5, group=Pulse, parameterName=Pulse Speed}
  n3 = Add(A=Sine(None=Multiply(A=@n1, B=@n2)), B=Multiply(A=ScalarParameter() {group=Pulse, parameterName=Sine Offset}) {constB=-1}) {constB=-1}
  n4 = ScalarParameter() {defaultValue=0.25, group=Pulse, parameterName=Pulse Scale}
  n5 = Multiply(A=Subtract(A=ObjectPositionWS(), B=WorldPosition().XYZ), B=Clamp(None=Multiply(A=@n3, B=@n4)) {clampMode=CMODE_ClampMax, maxDefault=0})
  n6 = Constant3Vector() {constant=<rgba 1 0.5 1 0>}
  LinearInterpolate(
   B=Multiply(A=Multiply(A=@n5, B=@n6) {desc=Halve Y Scale}, B=VertexColor()) {desc=Mask WPO by Vertex Coloring}
   Alpha=ScalarParameter() {defaultValue=1, parameterName=Turn On, sliderMax=1}
  ) {desc=On/Off Switch}
  n7 = TextureSample() {texture=/Game/Creative/Devices/MusicDevice/Materials/Textures/T_CP_Device_MusicDevice_Skin_01_D}
  LinearInterpolate(
   A=Multiply(A=VectorParameter() {parameterName=Base Color Tint, defaultValue=<rgba 0.24 0.2887 0.4062 1>}.RGB, B=@n7.RGB)
   B=@n7.RGB
   Alpha=@n7.RGB
  )
  n8 = Multiply(A=@n7.RGB, B=VectorParameter() {parameterName=Speaker Color Tint, defaultValue=<rgba 0.5544 0.8047 0.9297 1>}.RGB)
  n9 = ScalarParameter() {parameterName=Speaker Emissive Intensity}
  LinearInterpolate(B=Multiply(A=@n7.A, B=Multiply(A=@n8, B=@n9)), Alpha=ScalarParameter() {defaultValue=1, parameterName=Turn On, sliderMax=1}) {desc=On/Off Switch}
  Reroute(None=@n8)
  TextureSample() {samplerType=SAMPLERTYPE_Normal, texture=/Game/Creative/Devices/MusicDevice/Materials/Textures/T_CP_Device_MusicDevice_Skin_01_N}
  TextureSample() {samplerType=SAMPLERTYPE_Masks, texture=/Game/Creative/Devices/MusicDevice/Materials/Textures/T_CP_Device_MusicDevice_Skin_01_S}
 ]
}
